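# Material function that measures how far a point, world position by default, lies from the line through Start and End.
# It projects the point onto the normalized line direction with a dot product, without clamping to the segment, so the line is infinite.
# A static UseLocalSpace switch transforms both endpoints to world space first.
# It outputs RawDistance and an Alpha of that distance divided by MaxDistance and saturated, rising away from the line.
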MaterialFunction /WildEstate/TechArt/MaterialFunctions/MF_DistanceToLine
graph {nodes [
  n1 = FunctionInput(Preview=Constant3Vector() {desc=Top / End, constant=<rgba 0 0 64 0>}) {inputName=End, sortPriority=2}
  n2 = StaticSwitch(True=TransformPosition(None=@n1) {transformType=TRANSFORMPOSSOURCE_World}, False=@n1, Value=NamedRerouteUsage())
  n3 = FunctionInput(Preview=Constant3Vector() {desc=Bottom / Line Point}) {inputName=Start}
  n4 = StaticSwitch(True=TransformPosition(None=@n3) {transformType=TRANSFORMPOSSOURCE_World}, False=@n3, Value=NamedRerouteUsage())
  n5 = Normalize(VectorInput=Subtract(A=@n2, B=@n4)) {desc=Direction of Line}
  n6 = FunctionInput(Preview=WorldPosition().XYZ) {bUsePreviewValueAsDefault=true, desc=The point we are evaluating, inputName=WorldPosition, sortPriority=3}
  n7 = DotProduct(A=Subtract(A=@n6, B=@n4) {desc="Create non normalized line from start to camera\r\nto see how far away and what direction it is (?)"}, B=@n5)
  n8 = Distance(A=Add(A=@n4, B=Multiply(A=@n5, B=@n7)) {desc=New Pnt Location on Line to Sample}, B=@n6)
  FunctionOutput(None=@n8) {outputName=RawDistance, sortPriority=1}
  n9 = FunctionInput() {
   bUsePreviewValueAsDefault=true
   desc=The point we are evaluating
   inputName=MaxDistance
   inputType=FunctionInput_Scalar
   sortPriority=3
   previewValue=<500 0 0 1>
  }
  FunctionOutput(None=Saturate(None=Divide(A=@n8, B=@n9) {constB=200})) {outputName=Alpha}
  NamedRerouteDeclaration(None=Distance(A=@n2, B=@n4) {desc=Distance Between Points}) {name=PointDistance, nodeColor=<rgba 1 0.4219 0 1>}
  NamedRerouteDeclaration(None=FunctionInput(Preview=StaticBool()) {bUsePreviewValueAsDefault=true, inputName=UseLocalSpace, inputType=FunctionInput_StaticBool, sortPriority=4}) {name=Bool, nodeColor=<rgba 1 0 0.8672 1>}
 ]}
