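# Material function that makes a mesh face the camera as a billboard, outputting a position offset and world-space normals.
# Position takes UV direction from a center input, scales it by each vertex's distance to the object pivot and a Scale input, then converts camera space to world.
# Normals come from texture coordinates with constant depth, transformed from camera space without normalization.
# A yaw-only RotateAboutAxis branch and a spherical-depth calculation are left unconnected.
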MaterialFunction /WildEstate/TechArt/MaterialFunctions/MF_Billboard
graph {nodes [
  n1 = FunctionInput(Preview=TextureCoordinate()) {bUsePreviewValueAsDefault=true, inputName=UVs, inputType=FunctionInput_Vector2}
  n2 = ComponentMask(None=@n1) {g=true}
  n3 = Subtract(
   A=AppendVector(A=ComponentMask(None=@n1) {r=true}, B=OneMinus(None=@n2))
   B=FunctionInput(Preview=Constant2Vector() {g=0.5, r=0.5}) {bUsePreviewValueAsDefault=true, inputName=UV Center, inputType=FunctionInput_Vector2, sortPriority=4}
  ) {constB=0.5, desc=Make all verts equal distance to center}
  n4 = SquareRoot(None=DotProduct(A=@n3, B=@n3))
  n5 = Multiply(A=Divide(A=@n3, B=Max(A=@n4) {constB=0}), B=Distance(A=WorldPosition().XYZ, B=ObjectPositionWS()))
  n6 = FunctionInput() {bUsePreviewValueAsDefault=true, inputName=Scale, inputType=FunctionInput_Scalar, sortPriority=1, previewValue=<1 1 1 1>}
  n7 = Transform(None=AppendVector(A=Multiply(A=@n5, B=@n6), B=Constant())) {transformSourceType=TRANSFORMSOURCE_Camera}
  n8 = ObjectPositionWS()
  FunctionOutput(None=Subtract(A=Add(A=@n7, B=@n8), B=WorldPosition().XYZ)) {outputName=Position}
  n9 = Multiply(A=Subtract(A=TextureCoordinate()) {constB=0.5}) {constB=2}
  n10 = AppendVector(A=@n9, B=Constant() {r=1})
  n11 = Constant3Vector() {constant=<rgba 1 -1 -1 0>}
  FunctionOutput(None=Transform(None=Multiply(A=@n10, B=@n11)) {transformSourceType=TRANSFORMSOURCE_Camera}) {outputName="Normals (WS)", sortPriority=1}
  n12 = ObjectPositionWS()
  n13 = CameraPositionWS()
  n14 = Normalize(VectorInput=Subtract(A=ComponentMask(None=@n12) {g=true, r=true}, B=ComponentMask(None=@n13) {g=true, r=true}))
  n15 = Arctangent2Fast(Y=ComponentMask(None=@n14) {g=true}, X=ComponentMask(None=@n14) {r=true})
  n16 = Multiply() {constA=3.142, constB=2, desc=2 Pi}
  RotateAboutAxis(
   NormalizedRotationAxis=Constant3Vector() {constant=<rgba 0 0 1 0>}
   RotationAngle=Add(B=Divide(A=@n15, B=@n16) {desc=Converts Radians to 0-1}) {constA=0.25}
   PivotPoint=TransformPosition(None=Constant3Vector()) {transformType=TRANSFORMPOSSOURCE_World}
   Position=WorldPosition().XYZ
  )
  SquareRoot(None=OneMinus(None=DotProduct(A=@n9, B=@n9)))
 ]}
